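# Material function for the Valet vehicle that applies a cosmetic effect to incoming material attributes.
# It takes a Material input and an optional FXMask input, which falls back to a constant preview value.
# The shared MF_MetalLUT_TechArt function does the effect work through its Materials and Alpha inputs.
# Named reroutes label the input material, mask and output material and carry the wiring between stages.
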MaterialFunction /Game/Vehicles/Valet/CosmeticFX
graph {nodes [
  FunctionOutput(None=NamedRerouteUsage())
  NamedRerouteDeclaration(None=GetMaterialAttributes(None=FunctionInput() {inputName=Material, inputType=FunctionInput_MaterialAttributes}).MaterialAttributes) {name=InputMaterial, nodeColor=<rgba 0.1953 0 1 1>}
  NamedRerouteDeclaration(None=FunctionInput(Preview=Constant()) {bUsePreviewValueAsDefault=true, inputName=FXMask, inputType=FunctionInput_Vector4, sortPriority=1}) {name=FXMask, nodeColor=<rgba 0.125 1 0 1>}
  NamedRerouteDeclaration(None=MaterialFunctionCall!(Alpha=NamedRerouteUsage(), Materials=NamedRerouteUsage()) {materialFunction=/Game/Global/MaterialFunctions/TechArt/MF_MetalLUT_TechArt}.Result) {name=OutPutMaterial, nodeColor=<rgba 1 0.9141 0 1>}
 ]}
