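# Material function that corrects vertex normals to match vertices rotated with RotateAboutAxis, using the same axis and angle.
# It rotates the world-space normal about a pivot defaulting to zero, scales the resulting offset by a movement mask, adds it back and normalizes.
# Outputs are the raw offset, the world normal, and the tangent-space normal whole or split RG/B, intended for custom UVs and BlendAngleCorrectedNormals.
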MaterialFunction /Game/Vehicles/Valet/FixRotateAboutAxisNormals
graph {nodes [
  n1 = Constant()
  n2 = FunctionInput(Preview=VertexNormalWS()) {
   bUsePreviewValueAsDefault=true
   description=Filled by VertexNormalWS by default
   inputName=World Space Vertex Normals
   sortPriority=2
   previewValue=<1 0 0 1>
  }
  n3 = FunctionInput() {
   bUsePreviewValueAsDefault=true
   description="This value modulates the amount of normal deformation. Which means it can be used as a mask. "
   inputName=Movement Mask
   inputType=FunctionInput_Scalar
   sortPriority=3
   previewValue=<1 0 0 1>
  }
  n4 = Multiply(
   A=RotateAboutAxis(
    NormalizedRotationAxis=FunctionInput() {description=Use the same rotation axis used to rotate your models vertices., inputName=Rotation Axis}
    RotationAngle=FunctionInput() {description=Use the same rotation angle used to rotate your models vertices., inputName=Rotation Angle, sortPriority=1}
    PivotPoint=FunctionInput(Preview=@n1) {bUsePreviewValueAsDefault=true, inputName=PivotPoint, sortPriority=4}
    Position=@n2
   ) {desc=wo}
   B=@n3
  )
  FunctionOutput(None=@n4) {outputName=RotationOutputOnly}
  n5 = Normalize(VectorInput=Add(A=@n4, B=@n2))
  n6 = Transform(None=@n5) {transformSourceType=TRANSFORMSOURCE_World, transformType=TRANSFORM_Tangent}
  FunctionOutput(None=@n6) {
   description="It's recommended that this be calculated in the custom uvs input. The resulting output data should then be used as the \"BaseNormal\" in the BlendAngleCorrectedNormals material function. Your Normal map should be used as the \"AdditionalNormal\" input. The result should go into the material normal input. "
   outputName=New Tangent Space Vertex Normal
  }
  FunctionOutput(None=ComponentMask(None=@n6) {g=true, r=true}) {
   description="It's recommended that this be calculated in the custom uvs input. The resulting output data should then be used as the "
   outputName=New Tangent Space Vertex Normal RG
   sortPriority=1
  }
  FunctionOutput(None=ComponentMask(None=@n6) {b=true}) {
   description="It's recommended that this be calculated in the custom uvs input. The resulting output data should then be used as the "
   outputName=New Tangent Space Vertex Normal B
   sortPriority=2
  }
  FunctionOutput(None=@n5) {
   description="It's recommended that this be calculated in the custom uvs input. The resulting output data should then be used as the "
   outputName=New World Space Vertex Normal
   sortPriority=3
  }
 ]}
