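# Material function that animates the Valet vehicle's legged roof-mod turret through world position offset, switchable via UseVehicleModRoofAnim.
# Custom primitive data drives pitch, yaw, kickback, activation, weight shift and barrel overheat, with previous-frame values for motion vectors.
# Vertex-colour masks and curve-atlas rows shape pivot rotations for turret, barrel retraction and three legs, with matching normal correction.
# It also overrides input material attributes for optional status lights and overheat glow.
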MaterialFunction /Game/Vehicles/Valet/MF_ValetRoofModAnimation
graph {nodes [
  n1 = NamedRerouteUsage()
  n2 = ComponentMask(None=NamedRerouteUsage()) {b=true}
  n3 = Multiply(A=Time(), B=ScalarParameter() {defaultValue=0.125, group=VehicleMod_Anim, parameterName=AmbientLegMovementSpeed}) {constB=0.25}
  n4 = ConstantBiasScale(None=Sine(None=@n3))
  n5 = CurveAtlasRowParameter(CurveTime=@n4) {
   atlas=/Game/Vehicles/Valet/Textures/Turret_Atlas
   curve=/Game/Vehicles/Valet/Textures/LegAmbientAnimCurve
   defaultValue=2
   group=VehicleMod_Anim
   parameterName=LegAnimCurve
  }
  n6 = MaterialFunctionCall(A=@n5.R, B=@n5.R, C=@n5.R, Alpha=ChannelMaskParameter(None=NamedRerouteUsage()) {parameterName=LegVertColorMask}.R) {materialFunction=&Lerp_3Color}
  n7 = NamedRerouteDeclaration(None=@n5.A) {name=AmbientSquatCurve, nodeColor=<rgba 0.8281 1 0 1>}
  n8 = NamedRerouteUsage()
  n9 = ComponentMask(None=NamedRerouteUsage()) {g=true}
  n10 = LinearInterpolate(
   A=Multiply(A=LinearInterpolate(A=@n6.Result, B=@n7, Alpha=@n8), B=ScalarParameter() {defaultValue=0.1, group=VehicleMod_Anim, parameterName=LegMoveStr}) {constB=0.1}
   B=Multiply(A=ScalarParameter() {defaultValue=0.1, group=VehicleMod_Anim, parameterName=DebugInactiveFootRotation}, B=@n9)
   Alpha=NamedRerouteUsage()
  )
  n11 = MaterialFunctionCall!(Rotation Axis=NamedRerouteUsage(), Rotation Angle=@n10, World Space Vertex Normals=@n1, Movement Mask=@n2, PivotPoint) {materialFunction=&/FixRotateAboutAxisNormals}
  n12 = Transform(None=Constant3Vector() {constant=<rgba 0 0 1 0>}) {transformSourceType=TRANSFORMSOURCE_Local}
  n13 = CurveAtlasRowParameter(CurveTime=@n4) {
   atlas=&Turret_Atlas
   curve=/Game/Vehicles/Valet/Textures/LegAmbientAnimCurve3
   defaultValue=4
   group=VehicleMod_Anim
   parameterName=LegAnimCurve2
  }
  n14 = NamedRerouteUsage()
  n15 = Multiply(
   A=ScalarParameter() {defaultValue=0.15, group=VehicleMod_Anim, parameterName=LegMoveStr_02}
   B=MaterialFunctionCall(A=@n13.R, B=@n13.R, C=@n13.R, Alpha=ChannelMaskParameter(None=@n14) {parameterName=LegVertColorMask}.R) {materialFunction=&Lerp_3Color}.Result
  )
  Add(
   A=@n11.RotationOutputOnly
   B=MaterialFunctionCall(Rotation Axis=@n12, Rotation Angle=@n15, World Space Vertex Normals=Add(A=@n11.RotationOutputOnly, B=@n1), Movement Mask=@n2) {materialFunction=&FixRotateAboutAxisNormals}.RotationOutputOnly
  )
  n16 = NamedRerouteUsage()
  n17 = NamedRerouteUsage()
  n18 = NamedRerouteUsage()
  n19 = WorldPosition()
  n20 = RotateAboutAxis(
   NormalizedRotationAxis=NamedRerouteUsage()
   RotationAngle=@n10
   PivotPoint=LinearInterpolate(A=@n18, B=NamedRerouteUsage(), Alpha=Max(A=@n16, B=@n17))
   Position=@n19.XYZ
  )
  Add(A=@n20, B=RotateAboutAxis(NormalizedRotationAxis=@n12, RotationAngle=@n15, PivotPoint=@n18, Position=Add(A=@n20, B=@n19.XYZ)))
  Constant3Vector()
  Constant3Vector() {constant=<rgba -8.9609 0 45.9083 1>}
  n21 = Time()
  ConstantBiasScale(None=Sine(None=Multiply(A=@n21)))
  Frac(None=@n3)
  Frac(None=Multiply(A=Time(), B=ScalarParameter() {defaultValue=2, parameterName=DebugSpeed}))
  FunctionOutput(None=StaticSwitchParameter(True=NamedRerouteUsage(), False=NamedRerouteUsage()) {group=VehicleMod_Anim, parameterName=UseVehicleModRoofAnim})
  MaterialFunctionCall() {materialFunction=&ObjectPivotPoint}
  MaterialFunctionCall!(1, 2, 3, 4, A) {materialFunction=/Engine/Functions/Engine_MaterialFunctions02/Lerp_Multiple_Float}
  n22 = VectorParameter() {parameterName=OverheatEmissiveColor_02, defaultValue=<rgba 0.0543 0.0122 0.1302 1>}
  n23 = VectorParameter() {parameterName=OverheatEmissiveColor_01, defaultValue=<rgba 1 0.4251 0 1>}
  n24 = NamedRerouteUsage()
  n25 = Saturate(None=Multiply(A=@n24, B=@n24))
  n26 = Power(Base=Multiply(A=LinearInterpolate(A=@n22.RGB, B=@n23.RGB, Alpha=@n25), B=ScalarParameter() {defaultValue=5, parameterName=OverheatEmissive_Strength}))
  Multiply(A=@n26, B=LinearInterpolate(Alpha=NamedRerouteUsage()) {constA=0.35})
  n27 = Constant3Vector() {constant=<rgba 0.2656 0.2656 0.2656 0>}
  n28 = Constant() {r=1}
  n29 = Constant() {r=0.5}
  n30 = GetMaterialAttributes(None=FunctionInput(Preview=SetMaterialAttributes!(MaterialAttributes, BaseColor=@n27, Specular=@n28, Roughness=@n29) {attributeSetTypes [#1, #2, #3]}) {inputName=Material, inputType=FunctionInput_MaterialAttributes}) {attributeGetTypes [
    #4
    #5
    #1
    #2
    #3
   ]}
  NamedRerouteDeclaration(None=@n30.MaterialAttributes) {name=InputMaterial, nodeColor=<rgba 0 0.8828 1 1>}
  NamedRerouteDeclaration(None=@n30.Normal) {name=Input_Normal, nodeColor=<rgba 0.1016 1 0 1>}
  n31 = NamedRerouteUsage()
  n32 = NamedRerouteUsage()
  n33 = NamedRerouteUsage()
  n34 = NamedRerouteUsage()
  NamedRerouteDeclaration(None=Transform(None=MaterialFunctionCall(A=@n31, B=@n32, C=@n33, Alpha=@n34) {materialFunction=&Lerp_3Color}.Result) {transformSourceType=TRANSFORMSOURCE_Local}) {name=WeightedLegAxis, nodeColor=<rgba 1 0 0.6094 1>}
  n35 = NamedRerouteUsage()
  n36 = NamedRerouteUsage()
  n37 = NamedRerouteUsage()
  n38 = NamedRerouteUsage()
  NamedRerouteDeclaration(None=TransformPosition(None=MaterialFunctionCall(A=@n35, B=@n36, C=@n37, Alpha=@n38) {materialFunction=&Lerp_3Color}.Result) {transformType=TRANSFORMPOSSOURCE_World}) {name=WeightedLegPosition, nodeColor=<rgba 1 0.1172 0 1>}
  n39 = Saturate(None=ScalarParameter() {bUseCustomPrimitiveData=true, parameterName=W_WeightShift, primitiveDataIndex=4})
  NamedRerouteDeclaration(None=Multiply(A=@n39, B=ScalarParameter() {defaultValue=-0.05, group=VehicleMod_Anim, parameterName=WeightShift_Amt})) {name=DeltaTimeRotation, nodeColor=<rgba 1 0.6328 0 1>}
  n40 = NamedRerouteUsage()
  NamedRerouteDeclaration(None=SphereMask(A=ComponentMask(None=@n40) {g=true}, B=Constant() {r=0.5}) {attenuationRadius=0.1}) {name=NoYawMask, nodeColor=<rgba 0 1 0.25 1>}
  n41 = ComponentMask(None=NamedRerouteUsage()) {b=true}
  n42 = SphereMask(A=ChannelMaskParameter(None=NamedRerouteUsage()) {parameterName=LegVertColorMask}.R, B=NamedRerouteUsage(), Radius=Constant() {r=0.2})
  NamedRerouteDeclaration(None=OneMinus(None=Multiply(A=@n41, B=@n42))) {name=WeightShiftMask, nodeColor=<rgba 1 0 0.2578 1>}
  n43 = NamedRerouteUsage()
  n44 = NamedRerouteUsage()
  n45 = NamedRerouteUsage()
  n46 = Add(A=WorldPosition().XYZ, B=NamedRerouteUsage())
  NamedRerouteDeclaration(None=Multiply(A=RotateAboutAxis(NormalizedRotationAxis=@n43, RotationAngle=@n44, PivotPoint=@n45, Position=@n46), B=NamedRerouteUsage())) {name=WeightShiftWPO, nodeColor=<rgba 0 1 0.9297 1>}
  NamedRerouteDeclaration(None=VectorParameter() {group=VehicleMod_Anim, parameterName=Leg3_PivotPosition, defaultValue=<rgba -0.1324 -16.2534 8.8554 1>}.RGB) {name=Leg3PivotPosition, nodeColor=<rgba 1 0 0.2109 1>}
  NamedRerouteDeclaration(None=VectorParameter() {group=VehicleMod_Anim, parameterName=Leg2_PivotPosition, defaultValue=<rgba -13.6452 8.9032 8.663 1>}.RGB) {name=Leg2PivotPosition, nodeColor=<rgba 1 0.6562 0 1>}
  NamedRerouteDeclaration(None=VectorParameter() {group=VehicleMod_Anim, parameterName=Leg1_PivotPosition, defaultValue=<rgba 14.4485 7.9811 8.663 1>}.RGB) {name=Leg1PivotPosition, nodeColor=<rgba 0 1 1 1>}
  NamedRerouteDeclaration(None=@n30.WorldPositionOffset) {name=Input_WPO, nodeColor=<rgba 0 0.4609 1 1>}
  NamedRerouteDeclaration(None=VectorParameter() {group=VehicleMod_Anim, parameterName=Leg3RotationAxis, defaultValue=<rgba -1 0 0 1>}.RGB) {name=Leg3RotationAxis, nodeColor=<rgba 1 0.8672 0 1>}
  NamedRerouteDeclaration(None=VectorParameter() {group=VehicleMod_Anim, parameterName=Leg2RotationAxis, defaultValue=<rgba 0.5 0.866 0 1>}.RGB) {name=Leg2RotationAxis, nodeColor=<rgba 0 1 0.6016 1>}
  NamedRerouteDeclaration(None=VectorParameter() {group=VehicleMod_Anim, parameterName=Leg1RotationAxis, defaultValue=<rgba 0.5 -0.866 0 1>}.RGB) {name=Leg1RotationAxis, nodeColor=<rgba 0 1 0.8125 1>}
  n47 = NamedRerouteUsage()
  n48 = NamedRerouteUsage()
  NamedRerouteDeclaration(None=MaterialFunctionCall(
    Rotation Axis=NamedRerouteUsage()
    Rotation Angle=NamedRerouteUsage()
    World Space Vertex Normals=Add(A=@n47, B=@n48)
    Movement Mask=NamedRerouteUsage()
   ) {materialFunction=&FixRotateAboutAxisNormals}.RotationOutputOnly) {name=WeightShiftWPO_Normals, nodeColor=<rgba 0 1 0.1562 1>}
  n49 = ScalarParameter() {bUseCustomPrimitiveData=true, defaultValue=1, parameterName=W_Activate, primitiveDataIndex=3}
  n50 = ScalarParameter() {bUseCustomPrimitiveData=true, defaultValue=1, parameterName=W_Activate_Prev, primitiveDataIndex=9}
  n51 = Saturate(None=OneMinus(None=PreviousFrameSwitch(Current Frame=@n49, Previous Frame=@n50)))
  n52 = CurveAtlasRowParameter(CurveTime=@n51) {
   atlas=&Turret_Atlas
   curve=/Game/Vehicles/Valet/Textures/TurretActivateAnimCurve
   defaultValue=5
   group=VehicleMod_Anim
   parameterName=ActivateAnimCurve
  }
  NamedRerouteDeclaration(None=AppendVector(A=@n52, B=@n52)) {name=ActivateAnimCurves, nodeColor=<rgba 1 0.1875 0 1>}
  n53 = Subtract(A=WorldPosition() {worldPositionShaderOffset=WPT_ExcludeAllShaderOffsets}.XYZ, B=ObjectPositionWS())
  n54 = Ceil(None=ComponentMask(None=NamedRerouteUsage()) {b=true})
  n55 = Multiply(A=VectorParameter() {parameterName=Offset_TEmp, defaultValue=<rgba 0 0 -8 1>}.RGB, B=Saturate(None=OneMinus(None=@n54)))
  n56 = NamedRerouteUsage()
  n57 = CurveAtlasRowParameter(CurveTime=NamedRerouteDeclaration(None=@n39) {name=WeightShiftAnimCurve, nodeColor=<rgba 0.0547 0 1 1>}) {
   atlas=&Turret_Atlas
   curve=&TurretActivateAnimCurve
   defaultValue=5
   group=VehicleMod_Anim
   parameterName=ActivateAnimCurve
  }
  n58 = Max(A=Multiply(A=NamedRerouteUsage(), B=OneMinus(None=@n51)), B=Max(A=ComponentMask(None=@n56) {g=true}, B=@n57.R))
  NamedRerouteDeclaration(None=Add(A=Normalize(VectorInput=@n53), B=Multiply(A=@n55, B=@n58))) {name=Activate_WPO, nodeColor=<rgba 0.6172 1 0 1>}
  n59 = NamedRerouteUsage()
  n60 = ComponentMask(None=@n59) {g=true}
  n61 = OneMinus(None=SphereMask(A=@n60, B=Constant() {r=0.25}) {attenuationRadius=0.1})
  n62 = Multiply(A=Max(A=@n60, B=ComponentMask(None=@n59) {r=true}), B=@n61)
  NamedRerouteDeclaration(None=Multiply(A=NamedRerouteUsage(), B=@n62)) {name=AN_Kickback_Normal, nodeColor=<rgba 0 0.4141 1 1>}
  n63 = Subtract(
   A=WorldPosition().XYZ
   B=MaterialFunctionCall() {materialFunction=&ObjectPivotPoint}.Object Pivot Location
  )
  n64 = Transform(None=VectorParameter() {group=VehicleMod_Anim, parameterName=W_DirectionVector, defaultValue=<rgba 1 0 0 1>}.RGB) {transformSourceType=TRANSFORMSOURCE_Local}
  n65 = PreviousFrameSwitch(
   Current Frame=ScalarParameter() {bUseCustomPrimitiveData=true, parameterName=W_KickBack, primitiveDataIndex=1}
   Previous Frame=ScalarParameter() {bUseCustomPrimitiveData=true, parameterName=W_KickBack_Prev, primitiveDataIndex=10}
  )
  n66 = Divide(A=Subtract(A=Time(), B=@n65), B=ScalarParameter() {defaultValue=0.25, group=VehicleMod_Anim, parameterName=W_AnimDuration})
  n67 = CurveAtlasRowParameter(CurveTime=StaticSwitchParameter(True=Saturate(None=@n66), False=@n65) {group=VehicleMod_Anim, parameterName=UseTimeStamp}) {
   atlas=&Turret_Atlas
   curve=/Game/Vehicles/Valet/Textures/MachineGunAnimCurve
   group=VehicleMod_Anim
   parameterName=Turret_AnimCurve
  }
  n68 = Multiply(
   A=LinearInterpolate(
    A=ScalarParameter() {defaultValue=4, group=VehicleMod_Anim, parameterName=HullKickbackIntensity}
    B=ScalarParameter() {defaultValue=16, group=VehicleMod_Anim, parameterName=KickbackIntensity}
    Alpha=@n60
   )
   B=Subtract(A=Multiply(A=@n67.R) {constB=2})
  )
  n69 = Multiply(A=Add(A=Normalize(VectorInput=@n63), B=Multiply(A=@n64, B=@n68)), B=@n62)
  n70 = NamedRerouteUsage()
  NamedRerouteDeclaration(None=Multiply(A=Add(A=@n69, B=@n70), B=@n61)) {name=AN_Kickback_WPO, nodeColor=<rgba 1 0.7266 0 1>}
  n71 = Transform(None=VectorParameter() {group=VehicleMod_Anim, parameterName=W_LocalPitchAxis, defaultValue=<rgba 0 1 0 1>}.RGB) {transformSourceType=TRANSFORMSOURCE_Local}
  n72 = LinearInterpolate(
   A=PreviousFrameSwitch(
    Current Frame=ScalarParameter() {bUseCustomPrimitiveData=true, parameterName=W_Pitch}
    Previous Frame=ScalarParameter() {bUseCustomPrimitiveData=true, parameterName=W_Pitch_Prev, primitiveDataIndex=7}
   )
   B=ScalarParameter() {defaultValue=-0.15, group=VehicleMod_Anim, parameterName=InactivePitchValue}
   Alpha=ComponentMask(None=NamedRerouteUsage()) {r=true}
  )
  n73 = MaterialFunctionCall() {materialFunction=&ObjectPivotPoint}
  n74 = Transform(None=VectorParameter() {group=VehicleMod_Anim, parameterName=W_PitchPivotOffset, defaultValue=<rgba -10.5986 0 39.7765 1>}.RGB) {transformSourceType=TRANSFORMSOURCE_Local}
  n75 = NamedRerouteUsage()
  n76 = WorldPosition()
  NamedRerouteDeclaration(None=RotateAboutAxis(NormalizedRotationAxis=@n71, RotationAngle=@n72, PivotPoint=Add(A=@n73.Object Pivot Location, B=@n74), Position=Add(A=@n75, B=@n76.XYZ))) {name=AN_Pitch_WPO, nodeColor=<rgba 0 1 0.4375 1>}
  NamedRerouteDeclaration(None=VertexColor()) {name=VertexColor, nodeColor=<rgba 0.8516 1 0 1>}
  n77 = NamedRerouteUsage()
  n78 = NamedRerouteUsage()
  n79 = NamedRerouteUsage()
  NamedRerouteDeclaration(None=MaterialFunctionCall(Rotation Axis=@n71, Rotation Angle=@n72, World Space Vertex Normals=Add(A=@n77, B=@n78), Movement Mask=ComponentMask(None=@n79) {r=true}) {materialFunction=&FixRotateAboutAxisNormals}.RotationOutputOnly) {name=AN_Pitch_Normal, nodeColor=<rgba 0.3594 0 1 1>}
  n80 = Transform(None=VectorParameter() {group=VehicleMod_Anim, parameterName=W_LocalYawAxis, defaultValue=<rgba 0 0 1 1>}.RGB) {transformSourceType=TRANSFORMSOURCE_Local}
  n81 = LinearInterpolate(
   A=PreviousFrameSwitch(
    Current Frame=ScalarParameter() {bUseCustomPrimitiveData=true, parameterName=W_Yaw, primitiveDataIndex=2}
    Previous Frame=ScalarParameter() {bUseCustomPrimitiveData=true, parameterName=W_Yaw_Prev, primitiveDataIndex=8}
   )
   B=ScalarParameter() {group=VehicleMod_Anim, parameterName=InactiveYawValue}
   Alpha=ComponentMask(None=NamedRerouteUsage()) {r=true}
  )
  n82 = Add(
   A=MaterialFunctionCall() {materialFunction=&ObjectPivotPoint}.Object Pivot Location
   B=Transform(None=VectorParameter() {group=VehicleMod_Anim, parameterName=W_YawPivotOffset, defaultValue=<rgba 0 0 0 1>}.RGB) {transformSourceType=TRANSFORMSOURCE_Local}
  )
  n83 = Add(A=NamedRerouteUsage(), B=WorldPosition().XYZ)
  n84 = OneMinus(None=NamedRerouteUsage())
  NamedRerouteDeclaration(None=Multiply(A=RotateAboutAxis(NormalizedRotationAxis=@n80, RotationAngle=@n81, PivotPoint=@n82, Position=@n83), B=@n84)) {name=AN_Yaw_WPO, nodeColor=<rgba 1 0.3984 0 1>}
  n85 = Add(A=NamedRerouteUsage(), B=NamedRerouteUsage())
  NamedRerouteDeclaration(None=Multiply(
    A=MaterialFunctionCall(Rotation Axis=@n80, Rotation Angle=@n81, World Space Vertex Normals=@n85) {materialFunction=&FixRotateAboutAxisNormals}.RotationOutputOnly
    B=@n84
   )) {name=AN_Yaw_Normal, nodeColor=<rgba 0 1 0.1328 1>}
  n86 = Add(A=NamedRerouteUsage(), B=NamedRerouteUsage())
  n87 = ComponentMask(None=NamedRerouteUsage()) {r=true}
  n88 = Add(
   A=NamedRerouteDeclaration(None=Multiply(A=@n86, B=@n87) {constB=1.008}) {name=AN_Kickback+Pitch_WPO, nodeColor=<rgba 1 0 0.5625 1>}
   B=NamedRerouteUsage()
  )
  n89 = NamedRerouteUsage()
  n90 = ComponentMask(None=NamedRerouteUsage()) {b=true}
  n91 = Add(
   A=NamedRerouteDeclaration(None=LinearInterpolate(A=@n88, B=@n89, Alpha=@n90)) {name=WPO_Rotations, nodeColor=<rgba 0 1 0.3203 1>}
   B=NamedRerouteUsage()
  )
  n92 = NamedRerouteUsage()
  NamedRerouteDeclaration(None=Add(A=NamedRerouteUsage(), B=Add(A=@n91, B=@n92))) {name=FinalWPO, nodeColor=<rgba 1 0.7969 0 1>}
  n93 = NamedRerouteUsage()
  n94 = NamedRerouteDeclaration(None=LinearInterpolate(A=NamedRerouteUsage(), B=NamedRerouteUsage(), Alpha=ComponentMask(None=@n93) {b=true})) {name=WPO_Normals, nodeColor=<rgba 1 0 0.8203 1>}
  n95 = NamedRerouteUsage()
  n96 = NamedRerouteUsage()
  NamedRerouteDeclaration(None=Transform(None=LinearInterpolate(A=@n94, B=@n95, Alpha=@n96) {constB=0}) {transformSourceType=TRANSFORMSOURCE_World, transformType=TRANSFORM_Tangent}) {name=Final_Normal, nodeColor=<rgba 0 0.9297 1 1>}
  n97 = MaterialFunctionCall(Float3=NamedRerouteUsage()) {materialFunction=&BreakOutFloat3Components}
  NamedRerouteDeclaration(None=Multiply(A=@n97.R, B=@n97.R)) {name=BarrelMask, nodeColor=<rgba 1 0.5625 0 1>}
  n98 = TransformPosition(None=Add(A=Constant3Vector(), B=VectorParameter() {group=VehicleMod_Anim, parameterName=BarrelCollapseCenter, defaultValue=<rgba 5 0 56.3136 1>}.RGB)) {transformType=TRANSFORMPOSSOURCE_World}
  n99 = WorldPosition()
  n100 = VectorParameter() {group=VehicleMod_Anim, parameterName=W_DirectionVector, defaultValue=<rgba 1 0 0 1>}
  n101 = Add(
   A=Multiply(A=Subtract(A=@n98, B=@n99.XYZ), B=NamedRerouteUsage())
   B=Multiply(
    A=Transform(None=@n100.RGB) {transformSourceType=TRANSFORMSOURCE_Local}
    B=ScalarParameter() {defaultValue=-52, group=VehicleMod_Anim, parameterName=BarrelRetract_Offset}
   )
  )
  n102 = NamedRerouteUsage()
  n103 = NamedRerouteUsage()
  NamedRerouteDeclaration(None=Multiply(A=Multiply(A=@n101, B=@n102), B=ComponentMask(None=@n103) {b=true})) {name=AN_BarellRetract_WPO, nodeColor=<rgba 0.4766 0 1 1>}
  n104 = Subtract(
   A=WorldPosition().XYZ
   B=MaterialFunctionCall() {materialFunction=&ObjectPivotPoint}.Object Pivot Location
  )
  n105 = ComponentMask(None=Subtract(A=Transform(None=@n104) {transformSourceType=TRANSFORMSOURCE_World, transformType=TRANSFORM_Local}, B=NamedRerouteUsage())) {r=true}
  NamedRerouteDeclaration(None=Saturate(None=OneMinus(None=@n105))) {name=Barell_ClipMask, nodeColor=<rgba 0 1 0.5078 1>}
  n106 = NamedRerouteDeclaration(None=Add(A=NamedRerouteUsage(), B=NamedRerouteUsage())) {name=AN_Kickback+Pitch_Normal, nodeColor=<rgba 1 0.375 0 1>}
  n107 = NamedRerouteUsage()
  n108 = Normalize(VectorInput=Add(A=NamedRerouteUsage(), B=Add(A=@n106, B=@n107)))
  NamedRerouteDeclaration(None=@n108) {name=RecalculatedNormals, nodeColor=<rgba 0.5938 1 0 1>}
  NamedRerouteDeclaration(None=VectorParameter() {parameterName=BarrellClipMaskOffset, defaultValue=<rgba -5 0 -32 1>}.RGB) {name=Barrel_ClipMaskOffset, nodeColor=<rgba 0 1 0.6719 1>}
  n109 = NamedRerouteUsage()
  n110 = Constant() {r=0.1}
  n111 = NamedRerouteUsage()
  NamedRerouteDeclaration(None=StaticSwitchParameter(True=LinearInterpolate(A=@n109, B=@n110, Alpha=@n111), False=@n109) {parameterName=useEmissiveLights}) {name=Roughness, nodeColor=<rgba 1 0.7734 0 1>}
  n112 = ScalarParameter() {defaultValue=2, group=VehicleMod_Anim, parameterName=EmissiveStrength}
  n113 = Multiply(A=VectorParameter() {group=VehicleMod_Anim, parameterName=Activated_EmissiveColor, defaultValue=<rgba 0 0.5603 1 1>}.RGB, B=@n112)
  n114 = Constant3Vector() {constant=<rgba 0 0 0 1>}
  n115 = Multiply(A=VectorParameter() {group=VehicleMod_Anim, parameterName=DeActivated_EmissiveColor, defaultValue=<rgba 0.0104 0 0 1>}.RGB, B=@n112)
  n116 = MaterialFunctionCall(A=@n113, B=@n114, C=@n115, Alpha=ComponentMask(None=NamedRerouteUsage()) {a=true}) {materialFunction=&Lerp_3Color}
  n117 = NamedRerouteUsage()
  n118 = StaticSwitchParameter(True=Multiply(A=Multiply(A=@n116.Result, B=@n117), B=NamedRerouteUsage()), False=Constant3Vector()) {parameterName=useEmissiveLights}
  n119 = NamedRerouteUsage()
  NamedRerouteDeclaration(None=StaticSwitchParameter(True=Add(A=@n118, B=@n119), False=@n118) {parameterName=UseBarrellOverheat}) {name=Emissive, nodeColor=<rgba 0.3359 1 0 1>}
  n120 = NamedRerouteUsage()
  n121 = Constant() {r=1}
  n122 = NamedRerouteUsage()
  NamedRerouteDeclaration(None=StaticSwitchParameter(True=LinearInterpolate(A=@n120, B=@n121, Alpha=@n122), False=@n120) {parameterName=useEmissiveLights}) {name=Specular, nodeColor=<rgba 0 1 0.4141 1>}
  n123 = NamedRerouteUsage()
  n124 = NamedRerouteUsage()
  n125 = Multiply(
   A=MaterialFunctionCall(A=@n113, B=@n114, C=@n115, Alpha=ComponentMask(None=@n124) {a=true}) {materialFunction=&Lerp_3Color}.Result
   B=NamedRerouteUsage()
  )
  n126 = NamedRerouteUsage()
  NamedRerouteDeclaration(None=StaticSwitchParameter(True=LinearInterpolate(A=@n123, B=@n125, Alpha=@n126), False=@n123) {parameterName=useEmissiveLights}) {name=BaseColor, nodeColor=<rgba 0 0.6953 1 1>}
  NamedRerouteDeclaration(None=@n30.BaseColor) {name=InputBaseColor, nodeColor=<rgba 0 1 0.4609 1>}
  n127 = SetMaterialAttributes!(
   MaterialAttributes=NamedRerouteUsage()
   BaseColor=NamedRerouteUsage()
   Specular=NamedRerouteUsage()
   Roughness=NamedRerouteUsage()
   EmissiveColor=NamedRerouteUsage()
   Normal=NamedRerouteUsage()
   WorldPositionOffset=NamedRerouteUsage()
  ) {attributeSetTypes [
    #1
    #2
    #3
    #6
    #4
    #5
   ]}
  NamedRerouteDeclaration(None=@n127) {name=OutPutMaterial, nodeColor=<rgba 1 0 0.7266 1>}
  NamedRerouteDeclaration(None=@n30.Specular) {name=Input_Specular, nodeColor=<rgba 0.5469 1 0 1>}
  NamedRerouteDeclaration(None=@n20) {name=AmbientLegMovement, nodeColor=<rgba 1 0 0.9844 1>}
  NamedRerouteDeclaration(None=@n30.Roughness) {name=Input_Roughness, nodeColor=<rgba 1 0.6328 0 1>}
  n128 = Constant() {r=1}
  NamedRerouteDeclaration(None=ComponentMask(None=FunctionInput(Preview=@n128) {bUsePreviewValueAsDefault=true, inputName=EmissiveMask, sortPriority=1}) {r=true}) {name=Emissive_Mask, nodeColor=<rgba 0 1 0.3906 1>}
  n129 = OneMinus(None=Fresnel(
    ExponentIn=ScalarParameter() {defaultValue=0.5, parameterName=Emissive_Exponent}
    BaseReflectFractionIn=ScalarParameter() {defaultValue=0.04, parameterName=Emissive_ReflectRefracIn}
    Normal=PixelNormalWS()
   ))
  n130 = ScalarParameter() {defaultValue=2, parameterName=EmissiveFresnelPow}
  NamedRerouteDeclaration(None=Saturate(None=Power(Base=@n129, Exp=@n130))) {name=EmissiveFrenselMask, nodeColor=<rgba 1 0.0703 0 1>}
  n131 = MaterialFunctionCall(Float3=NamedRerouteUsage()) {materialFunction=&BreakOutFloat3Components}
  n132 = OneMinus(None=@n131.B)
  n133 = LocalPosition() {localOrigin=InstancePreSkinning}
  n134 = PreSkinnedLocalBounds()
  n135 = Subtract(A=ShaderStageSwitch(PixelShader=VertexInterpolator(VS=@n133.XYZ).PS, VertexShader=@n133.XYZ), B=@n134.Min)
  n136 = VectorParameter() {parameterName=BarrelHeatGradientOffset, defaultValue=<rgba -32 0 0 1>}
  n137 = Saturate(None=Divide(A=Add(A=@n135, B=@n136.RGB), B=@n134.Max))
  n138 = OneMinus(None=ScalarParameter() {bUseCustomPrimitiveData=true, parameterName=W_Barrell_Overheat, primitiveDataIndex=5})
  n139 = Saturate(None=Subtract(A=ComponentMask(None=@n137) {r=true}, B=Saturate(None=@n138)))
  NamedRerouteDeclaration(None=Multiply(A=Multiply(A=@n131.R, B=@n132), B=NamedRerouteDeclaration(None=@n139) {name=Barrell_Gradient, nodeColor=<rgba 0 0.9062 1 1>})) {name=Barrell_GradientMask, nodeColor=<rgba 1 0 0.1406 1>}
  NamedRerouteDeclaration(None=Multiply(A=@n26, B=NamedRerouteUsage())) {name=OverheatEmissive, nodeColor=<rgba 1 0 0.8437 1>}
  n140 = VectorParameter() {group=VehicleMod_Anim, parameterName=Leg3_Squat_PivotPosition, defaultValue=<rgba -0.6271 -30.9038 0 1>}
  NamedRerouteDeclaration(None=@n140.RGB) {name=Leg3_SquatPivot, nodeColor=<rgba 1 0 0.2578 1>}
  n141 = VectorParameter() {group=VehicleMod_Anim, parameterName=Leg2_Squat_PivotPosition, defaultValue=<rgba -25.8522 16.0073 0 1>}
  NamedRerouteDeclaration(None=@n141.RGB) {name=Leg2_SquatPivot, nodeColor=<rgba 0 0.5781 1 1>}
  NamedRerouteDeclaration(None=VectorParameter() {group=VehicleMod_Anim, parameterName=Leg1_Squat_PivotPosition, defaultValue=<rgba 26.9219 15.1088 0 1>}.RGB) {name=Leg1_SquatPivot, nodeColor=<rgba 0.8984 1 0 1>}
  n142 = NamedRerouteUsage()
  n143 = NamedRerouteUsage()
  n144 = NamedRerouteUsage()
  n145 = ChannelMaskParameter(None=NamedRerouteUsage()) {parameterName=LegVertColorMask}
  NamedRerouteDeclaration(None=TransformPosition(None=MaterialFunctionCall(A=@n142, B=@n143, C=@n144, Alpha=@n145.R) {materialFunction=&Lerp_3Color}.Result) {transformType=TRANSFORMPOSSOURCE_World}) {name=Squat_LegPivotPositions, nodeColor=<rgba 0.7813 0 1 1>}
  n146 = Normalize(VectorInput=Add(A=NamedRerouteUsage(), B=@n11.RotationOutputOnly))
  NamedRerouteDeclaration(None=@n146) {name=AmbientLegMovementVertexNormal, nodeColor=<rgba 0.6406 1 0 1>}
  n147 = NamedRerouteUsage()
  n148 = ScalarParameter() {defaultValue=0.1, parameterName=SquatLegRotation}
  NamedRerouteDeclaration(None=RotateAboutAxis(
    NormalizedRotationAxis=NamedRerouteUsage()
    RotationAngle=Multiply(A=@n147, B=@n148)
    PivotPoint=NamedRerouteUsage()
    Position=WorldPosition().XYZ
   )) {name=Anim_SquatLegRotation, nodeColor=<rgba 0 1 0.4609 1>}
  n149 = NamedRerouteUsage()
  NamedRerouteDeclaration(None=Saturate(None=Ceil(None=@n149))) {name=SquatCurveShift, nodeColor=<rgba 0 0.9766 1 1>}
  NamedRerouteDeclaration(None=Ceil(None=@n9)) {name=ActivateCurveShift, nodeColor=<rgba 1 0.7031 0 1>}
  NamedRerouteDeclaration(None=Transform(None=NamedRerouteUsage())) {name=InputNormalWS, nodeColor=<rgba 0 1 0.0391 1>}
  n150 = NamedRerouteUsage()
  n151 = NamedRerouteUsage()
  n152 = NamedRerouteUsage()
  n153 = ChannelMaskParameter(None=NamedRerouteUsage()) {parameterName=LegVertColorMask}
  NamedRerouteDeclaration(None=TransformPosition(None=MaterialFunctionCall(A=@n150, B=@n151, C=@n152, Alpha=@n153.R) {materialFunction=&Lerp_3Color}.Result) {transformType=TRANSFORMPOSSOURCE_World}) {name=LegPivotPositions, nodeColor=<rgba 0.4062 0 1 1>}
  n154 = NamedRerouteUsage()
  n155 = NamedRerouteUsage()
  n156 = NamedRerouteUsage()
  n157 = ChannelMaskParameter(None=NamedRerouteUsage()) {parameterName=LegVertColorMask}
  NamedRerouteDeclaration(None=Transform(None=MaterialFunctionCall(A=@n154, B=@n155, C=@n156, Alpha=@n157.R) {materialFunction=&Lerp_3Color}.Result) {transformSourceType=TRANSFORMSOURCE_Local}) {name=LegsDirectionVector, nodeColor=<rgba 0 0.3672 1 1>}
  NamedRerouteDeclaration(None=ScalarParameter() {defaultValue=0.5, group=VehicleMod_Anim, parameterName=PickLeg}) {name=Weightedleg, nodeColor=<rgba 1 0 0.4688 1>}
  NamedRerouteUsage()
  NamedRerouteUsage()
  NamedRerouteUsage()
  NamedRerouteUsage()
  ObjectPositionWS()
  ScalarParameter() {bUseCustomPrimitiveData=true, parameterName=SquatDebug, primitiveDataIndex=6}
  Transform(None=@n108) {transformSourceType=TRANSFORMSOURCE_World, transformType=TRANSFORM_Tangent}
  Transform(None=@n146) {transformSourceType=TRANSFORMSOURCE_World, transformType=TRANSFORM_Tangent}
  VertexNormalWS()
  VertexNormalWS()
  VertexNormalWS()
  VertexNormalWS()
  VertexNormalWS()
  VertexNormalWS()
 ]}
