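# Material function that layers scalar-driven paint damage onto incoming material attributes for the Valet vehicle.
# A tiled AO texture, boosted by Damage and sharpened by a power, masks chips that shift color, metallic, roughness and clearcoat channels.
# Damage also blends in an angle-corrected damage normal and scales most changes, though clearcoat channels follow only the chip mask.
# Bounds-based quadrant masks, per-corner damage parameters and a shading-model switch sit unconnected to the output.
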MaterialFunction /Game/Vehicles/Valet/MF_Valet_Damage
graph {nodes [
  Add()
  Constant()
  Constant()
  n1 = GetMaterialAttributes(None=FunctionInput() {inputType=FunctionInput_MaterialAttributes}) {attributeGetTypes [
    #1
    #2
    #3
    #4
    #5
    #6
   ]}
  n2 = VectorParameter() {group=Damage, parameterName=DamageChips, defaultValue=<rgba 1 1 1 1>}
  n3 = Multiply(A=TextureCoordinate(), B=ScalarParameter() {defaultValue=1.6, group=Damage, parameterName=Damage Tiling})
  n4 = TextureSample(UVs=@n3) {samplerType=SAMPLERTYPE_LinearGrayscale, texture=&/ValetDamage_AO}
  n5 = ScalarParameter() {group=Damage, parameterName=Damage}
  n6 = Saturate(None=@n5)
  n7 = LinearInterpolate(B=ScalarParameter() {defaultValue=1.1, group=Damage, parameterName=Chip Amount Max, sliderMin=1.5}, Alpha=@n6) {constA=0.8, constB=1.1}
  n8 = Clamp(None=@n7) {maxDefault=1.25}
  n9 = Saturate(None=Power(Base=Multiply(A=@n4.G, B=@n8)) {constExponent=16})
  n10 = Saturate(None=LinearInterpolate(A=@n1.BaseColor, B=FunctionInput(Preview=@n2.RGB) {bUsePreviewValueAsDefault=true, inputName=Metal Tint}, Alpha=@n9))
  n11 = LinearInterpolate(A=@n1.Metallic, B=Constant() {r=1}, Alpha=@n9)
  n12 = TextureSample(UVs=@n3) {samplerType=SAMPLERTYPE_Normal, texture=&/ValetDamage_N}
  n13 = ScalarParameter() {group=Damage, parameterName=NormalFlatness}
  n14 = LinearInterpolate(
   A=@n1.Normal
   B=MaterialFunctionCall(
    BaseNormal=@n1.Normal
    AdditionalNormal=MaterialFunctionCall(Normal=@n12.RGB, Flatness=@n13) {materialFunction=&FlattenNormal}.Result
   ) {materialFunction=&BlendAngleCorrectedNormals}.Result
   Alpha=@n6
  )
  n15 = Saturate(None=Multiply(A=@n1.Roughness, B=LinearInterpolate(Alpha=@n4.G) {constA=0.75, constB=1.5}))
  n16 = Constant() {r=0.1}
  n17 = LinearInterpolate(A=@n1.Roughness, B=Clamp(None=LinearInterpolate(A=@n15, B=@n16, Alpha=@n9)) {maxDefault=0.9, minDefault=0.1}, Alpha=@n6)
  n18 = LinearInterpolate(A=@n1.CustomData0, B=ScalarParameter() {group=Damage, parameterName=Clearcoat?}, Alpha=@n9)
  n19 = SetMaterialAttributes!(
   MaterialAttributes=@n1.MaterialAttributes
   BaseColor=Saturate(None=LinearInterpolate(A=@n1.BaseColor, B=@n10, Alpha=@n6))
   Metallic=Saturate(None=LinearInterpolate(A=@n1.Metallic, B=@n11, Alpha=@n6))
   Normal=@n14
   Roughness=Clamp(None=Saturate(None=@n17)) {minDefault=0.01}
   CustomData0=@n18
   CustomData1=LinearInterpolate(A=@n1.CustomData1, Alpha=@n9)
  ) {attributeSetTypes [
    #1
    #2
    #3
    #4
    #5
    #6
   ]}
  FunctionOutput(None=@n19)
  If(A=@n18, B=Constant() {r=0.75}, "A > B"=ShadingModel(), "A < B"=ShadingModel() {shadingModel=MSM_ClearCoat})
  LinearInterpolate()
  n20 = VertexInterpolator(VS=PreSkinnedPosition())
  n21 = PreSkinnedLocalBounds()
  n22 = FunctionInput(Preview=Divide(A=Subtract(A=@n20.PS, B=@n21.Min), B=Subtract(A=@n21.Max, B=@n21.Max))) {inputName=Local Bounds Split}
  n23 = ScalarParameter() {defaultValue=1, group=Damage, parameterName=Contrast}
  n24 = MaterialFunctionCall(In=ComponentMask(None=@n22) {r=true}, Contrast=@n23) {materialFunction=&CheapContrast}
  n25 = Saturate(None=@n24.Result)
  n26 = MaterialFunctionCall(In=ComponentMask(None=@n22) {g=true}, Contrast=@n23) {materialFunction=&CheapContrast}
  n27 = Saturate(None=@n26.Result)
  n28 = Multiply(A=@n25, B=@n27)
  n29 = Multiply(A=@n25, B=OneMinus(None=@n27))
  n30 = Saturate(None=OneMinus(None=@n24.Result))
  n31 = Saturate(None=@n26.Result)
  n32 = Multiply(A=@n30, B=@n31)
  n33 = Multiply(A=@n30, B=OneMinus(None=@n31))
  Max(A=Max(A=Multiply(A=@n5, B=@n28), B=Multiply(A=@n5, B=@n29)), B=Max(A=Multiply(A=@n5, B=@n32), B=Multiply(A=@n5, B=@n33)))
  Normalize(VectorInput=@n14)
  Saturate(None=@n7)
  Saturate()
  ScalarParameter() {group=Damage, parameterName=FrontLeft_Damage}
  ScalarParameter() {group=Damage, parameterName=FrontRight_Damage, sliderMax=1}
  ScalarParameter() {group=Damage, parameterName=BackLeft_Damage}
  ScalarParameter() {group=Damage, parameterName=BackRight_Damage}
 ]}
