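# Parent material for Valet vehicle mods, a lit material built on material attributes with parameterized textures, also usable on Niagara mesh particles.
# Diffuse, normal and emissive textures feed the attributes, with the SRM texture's green channel driving metallic, specular and roughness alike.
# World position offset comes from a hit-and-shrink function using custom primitive data.
# The attributes then pass through cosmetic FX, roof-mod animation, Valet damage and deploy wobble functions.
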Material /Game/Vehicles/Valet/M_ValetMods_Parent
graph {
 properties {bUseMaterialAttributes=true, bUsedWithNiagaraMeshParticles=true, pixelDepthOffsetMode=PDOM_Legacy, usedShadingModels=Default Lit}
 nodes [
  n1 = TextureSampleParameter2D() {parameterName=SRM_Texture, texture=/Engine/EngineResources/Black_Low.Black}
  n2 = TextureSampleParameter2D() {parameterName=Emissive, texture=&Black}
  n3 = SetMaterialAttributes!(
   MaterialAttributes
   Base Color=TextureSampleParameter2D() {parameterName=Diffuse, texture=&Black}.RGB
   Metallic=@n1.G
   Specular=@n1.G
   Roughness=@n1.G
   Emissive Color=@n2.RGB
   Normal=TextureSampleParameter2D() {parameterName=Normal, samplerType=SAMPLERTYPE_Normal, texture=&FlatNormalBlue}.RGB
   World Position Offset=MaterialFunctionCall!(
    WSDeformationCenter=VectorParameter() {bUseCustomPrimitiveData=true, group=z_code, parameterName=WSDeformationCenter, primitiveDataIndex=20}.RGB
    Code_DestructionAmount=ScalarParameter() {group=z_code, parameterName=Code_DestructionAmount}
    Code_UseUvForDamage
    Code_UseHitEffects=StaticBoolParameter() {defaultValue=true, group=z_code, parameterName=Code_UseHitEffects}
    Code_Emissive
    WSDamageRadius=ScalarParameter() {bUseCustomPrimitiveData=true, defaultValue=1, group=z_code, parameterName=WSDamageRadius, primitiveDataIndex=26}
    CrackDamageShader
    Diffuse
    ScalarModDamageCracks
    ScalarOffsetforDamageCracks
    ExposeLerpFunctionForInvertingTheHitDistanceCalc
    InvertDamageFalloff
    CrackWorldPosition
    WorldPosition
    DistanceCullFadeAnimationOffset=ScalarParameter() {bUseCustomPrimitiveData=true, group=z_code, parameterName=DistanceCullFadeAnimationOffset, primitiveDataIndex=12}
   ) {materialFunction=/Game/Packages/jl_tests/Fortnite_HitEffects}.Hit And Shrink WPO
  ) {attributeSetTypes [
    #1
    #2
    #3
    #4
    #5
    #6
    #7
   ]}
  n4 = MaterialFunctionCall!(Material=@n3, FXMask=NamedRerouteUsage()) {materialFunction=&/CosmeticFX}
  n5 = NamedRerouteUsage()
  MaterialFunctionCall!(In=MaterialFunctionCall!(
    In=MaterialFunctionCall!(Material=@n4.Result, EmissiveMask=@n5) {materialFunction=&/MF_ValetRoofModAnimation}.Result
    Metal Tint
    Local Bounds Split
   ) {materialFunction=&/MF_Valet_Damage}.Result) {materialFunction=/Game/Global/MaterialFunctions/MF_DeployWobble_Simple}
  NamedRerouteDeclaration(None=@n2.RGB) {name=EmissiveTexture, nodeColor=<rgba 0.5234 1 0 1>}
  NamedRerouteDeclaration(None=TextureSampleParameter2D() {parameterName=FX_Mask, texture=&Black}.RGBA) {name=FXMaskTexture, nodeColor=<rgba 1 0 0.8437 1>}
  Reroute(None=@n2.RGB)
 ]
}
